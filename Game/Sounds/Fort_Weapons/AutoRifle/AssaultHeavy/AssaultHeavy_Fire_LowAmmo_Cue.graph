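# Sound cue for the low-ammo fire sound of the heavy assault rifle.
# A single wave player plays one low-ammo fire sample at slightly reduced pitch and volume, with no randomization.
# It routes to a degraded ranged-weapon low-ammo sound class and joins local and global weapon gunfire concurrency sets.
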SoundCue /Game/Sounds/Fort_Weapons/AutoRifle/AssaultHeavy/AssaultHeavy_Fire_LowAmmo_Cue
root=WavePlayer_1
nodes [{id=WavePlayer_1, properties {soundWaveAssetPtr=&/AssaultHeavy_Fire_LowAmmo_02}}]
properties {
 maxDistance=2097152
 pitchMultiplier=0.9
 soundClassObject=/Game/Sounds/FortSoundClassesAndModes/Fort_WeaponRanged_LowAmmo_Degraded_PSM
 volumeMultiplier=0.8
 concurrencyOverrides {resolutionRule=PreventNew}
 concurrencySet [&Weapon_Gunfire_Local_Conc, &Weapon_Gunfire_Global_Conc]
}
