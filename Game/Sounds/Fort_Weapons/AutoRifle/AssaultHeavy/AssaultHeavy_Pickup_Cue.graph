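# Sound cue for picking up the heavy assault rifle, routed to the Fort_Loot_Collect sound class with weapon pickup attenuation.
# A single wave player feeds a short fixed delay at the root, with volume and pitch lowered on the cue.
# A concurrency override allows one instance at a time and blocks new plays while it is active.
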SoundCue /Game/Sounds/Fort_Weapons/AutoRifle/AssaultHeavy/AssaultHeavy_Pickup_Cue
root=Delay_2
nodes [
 {id=Delay_2, children [WavePlayer_2], properties {delayMax=0.07, delayMin=0.07}}
 {id=WavePlayer_2, properties {soundWaveAssetPtr=&/AssaultHeavy_Pickup_01}}
]
properties {
 attenuationSettings=&Fort_Weapon_Pickup_Att
 bOverrideConcurrency=true
 maxDistance=3500
 pitchMultiplier=0.9
 soundClassObject=&Fort_Loot_Collect
 volumeMultiplier=0.2077
 concurrencyOverrides {maxCount=1, resolutionRule=PreventNew}
}
