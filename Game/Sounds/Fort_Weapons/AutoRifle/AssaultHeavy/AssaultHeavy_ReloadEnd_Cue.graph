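# Sound cue for the end-of-reload sound of the AssaultHeavy auto rifle weapon.
# A single wave player feeds the root directly, with no randomization or modulation.
# It routes through the weapon foley sound class and the shared weapon reload attenuation, at slightly reduced volume.
# The cue overrides concurrency and has its debug flag left on.
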SoundCue /Game/Sounds/Fort_Weapons/AutoRifle/AssaultHeavy/AssaultHeavy_ReloadEnd_Cue
root=WavePlayer_0
nodes [{id=WavePlayer_0, properties {soundWaveAssetPtr=&/AssaultHeavy_ReloadEnd_01}}]
properties {
 attenuationSettings=&Fort_Weapon_Reload_Att
 bDebug=true
 bOverrideConcurrency=true
 maxDistance=2250
 soundClassObject=&Fort_WeaponFoley
 volumeMultiplier=0.752
}
